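; Gameplay ability for the giant's finisher on the stunned bug Queen boss.
; It switches secondary pilots to third-person view, flags a target found by targeting or a server distance fallback, then plays a montage.
; Montage notifies trigger sword power-up, lethal impact with area damage, and a timed power-level buff, using client-to-server RPCs and once-only gates against missed notifies.
; Ending clears effects and broadcasts completion and quest events.
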(event Ability|EventActivateAbilityFromEvent
  (params (EventData :type Gameplay%20Event%20Data%20Structure%20%28by%20ref%29))
  (bind _returnvalue (Fortnite|Athena|GameState|GetGameStateAthena))
  ;[0 0.15 0.001972 0.5] ActivateAbilityFromEvent: Playing out the finisher animation.
  (|CallParentFunction EventData)
  (Fortnite|FortShippingLog self "ActivateAbility: Invoked.")
  (Utilities|FlowControl|Sequence
    (:then_0
      ;;[0.15 0 0.1345 0.5] Apply the finisher active gameplay effect to owner.
      (bind _returnvalue_2 (Ability|ApplyGameplayEffectToOwner (Variables|Settings|GetFinisherActiveGameplayEffectClass)))
      (Variables|Transient|SetFinisherActiveGameplayEffectHandle _returnvalue_2)
      ;;[0 0.15 0.1356 0.5] Reset state.
      (Variables|Transient|SetbFinisherApplied false)
      (Variables|Transient|SetbBuffApplied false)
      (Variables|Transient|SetPowerUpApplied false)
      ;;[0 0.15 0.06525 0.5] Cache the game state component.
      (Utilities|IsValid _returnvalue
        (:Is%20Valid
          (bind _output_get (Variables|Transient|SetGameStateComponent (Actor|GetComponentbyClass _returnvalue "/ToyFoxGameplay/Gameplay/B_ToyFoxGameStateComponent.B_ToyFoxGameStateComponent_C")))
          ;;[0.1148 0 0.15 0.5] Apply gameplay effects to secondary pilots so they enter the finisher viewing state.
          (Ability|HasAuthority
            (:True
              (Utilities|IsValid _output_get
                (:Is%20Valid
                  (Class|BToyFoxGameStateComponent|SetSecondaryPilot3PView _output_get true))))))))
    (:then_1
      ;;[0.15 0 0.1304 0.5] Let the bug know that the finisher is going to happen. This prevents the bug from recoverying from the stunned state.
      (bind (_asynctaskproxy _targetdata) (Ability|Tasks|PerformTargetingAbilityTask "FinisherTargetingTask" (Variables|Settings|GetFinisherTargetingPreset) false false true)
        (:then
          (Variables|Transient|SetAsyncTask_FinisherTargeting _asynctaskproxy))
        (:Targeted
          _)
        (:Cancelled
          _))
      (for _array_element (Ability|TargetData|GetAllActorsfromTargetData _targetdata)
        (bind _asbuilding_actor (Utilities|Casting|CastToBuildingActor _array_element))
        (if (GameplayTags|HasAllMatchingGameplayTags _asbuilding_actor (Variables|Settings|GetFinisherValidGameplayTagContainer))
          (bind _output_get_2 (Variables|Transient|SetFinisherTargetActor _asbuilding_actor))
          (Fortnite|FortShippingLog self (Utilities|String|Append "FinisherTargetingTask found finisher target actor: " (Utilities|GetDisplayName _output_get_2)))
          (Ability|ApplyGameplayEffectToTarget (Ability|TargetData|AbilityTargetDatafromActor (Variables|Transient|GetFinisherTargetActor)) (Variables|Settings|GetIncomingFinisherGameplayEffectClass))))
      (Variables|Transient|GetFinisherTargetActor
        (:then)
        (:else
          (Fortnite|FortShippingLog self "Ending ability, FinisherTargetActor is invalid." true)
          (bind (_returnvalue_3 _targetactor) (CallFunction|FallbackTargetDetection))
          (if _returnvalue_3
            (bind _output_get_3 (Variables|Transient|SetFinisherTargetActor _targetactor))
            (Fortnite|FortShippingLog self (Utilities|String|Append "Fallback logic found finisher target actor: " (Utilities|GetDisplayName _output_get_3)))
            (else
              (Ability|EndAbility))))))
    (:then_2
      ;;[0.06094 0 0.09375 0.5] Play the finisher montage.
      (bind (_asynctaskproxy_2 _notifytag) (Ability|Tasks|PlayMontageAndWaitNotify "FinisherMontageTask" (Variables|Settings|GetFinisherMontage))
        (:OnCompleted
          _)
        (:OnBlendOut
          _)
        (:OnInterrupted
          _)
        (:OnCancelled
          _)
        (:OnNotifyBegin
          (if (== _notifytag "(TagName=\"ToyFox.AnimNotify.Ability.Finisher.Sword.Powerup\")")
            (Fortnite|FortShippingLog self "NotifyTag: Sword.Powerup")
            (Ability|HasAuthority
              (:True
                _)
              (:False
                (CallFunction|ServerApplyPowerUp) ; Client is invoking apply power up RPC since server \ncould miss anim notify sometime
              ))
            (CallFunction|ApplyPowerUp)
            (elif (== _notifytag "(TagName=\"ToyFox.AnimNotify.Ability.Finisher.Impact\")")
              ;;;[0.1298 0.07475 0.15 0.5] Finisher impact notify. Let the target (bug) actor know that it is time to die out.
              (Fortnite|FortShippingLog self "NotifyTag: Finisher.Impact")
              (Ability|HasAuthority
                (:True
                  _)
                (:False
                  (CallFunction|ServerApplyFinisher) ; Client is invoking an apply finisher RPC since the server for \nsome reason can miss this anim notify.
                ))
              (CallFunction|ApplyFinisher)
              (elif (== _notifytag "(TagName=\"ToyFox.AnimNotify.Ability.Finisher.Buff\")")
                ;;;[0.15 0 0.1169 0.5] This is when the power level buff gets applied to the giant.
                (Fortnite|FortShippingLog self "NotifyTag: Buff.")
                (Ability|HasAuthority
                  (:True
                    _)
                  (:False
                    (CallFunction|ServerApplyBuff) ; Client is invoking an apply buff RPC since the server for \nsome reason can miss this anim notify.
                  ))
                (CallFunction|ApplyBuff))))))
      (if (Ability|GameplayAbility|IsActive 0)
        (Fortnite|FortShippingLog self "EndAbility: Invoked.")
        (else
          _))
      (Ability|EndAbility))))

(event Custom|ServerApplyBuff
  ;[0.15 0.01221 4.5e-05 0.5] ApplyBuff: Provide the giant a buff to its power level.
  (Ability|HasAuthority
    (:True
      (if (Ability|GameplayAbility|IsActive 0)
        (Utilities|FlowControl|DeltaGateWildcard (Variables|Transient|GetbBuffApplied) true
          (:Success
            (Fortnite|FortShippingLog self "Applying Buff")
            ;;[0.03102 0.08319 0.15 0.5] Handle behavior during the buff moment.
            (if (ScalableFloat|GetValueasBool (Variables|Hotfix|GetHotfix_BuffGameplayEnabled))
              (bind _returnvalue (Ability|ApplyGameplayEffectToOwner (Variables|Settings|GetFinisherBuffGameplayEffectClass)))
              (Variables|Transient|SetFinisherBuffEffectHandle _returnvalue)
              (Ability|HasAuthority
                (:True
                  (bind _returnvalue_2 (Utilities|Time|SetTimerbyEvent (EventDispatchers|CreateEvent :!event Custom|HandlePowerLevelRecoveryInterval) (ScalableFloat|GetCachedScalableFloatValue (Variables|Hotfix|GetHotfix_PowerLevelRecoveryInterval)) true))
                  (Variables|Transient|SetPowerLevelRecoveryTimerHandle _returnvalue_2)
                  (CallFunction|HandlePowerLevelRecoveryInterval))))))))))

(event Custom|ServerApplyPowerUp
  ; Apply PowerUp
  (Ability|HasAuthority
    (:True
      (if (Ability|GameplayAbility|IsActive 0)
        (Utilities|FlowControl|DeltaGateWildcard (Variables|Transient|GetPowerUpApplied) true
          (:Success
            (Fortnite|FortShippingLog self "Apply Powerup")
            ;;[0.003337 0.06572 0.15 0.5] Start powering up the sword here.
            (Ability|ExecuteGameplayCueOnOwner (Utilities|Struct|BreakGameplayCueTag (Variables|GameplayCueTag|GetGameplayCueTag_SwordPowerup)))))))))

(event Custom|ApplyBuff
  ;[0.15 0.01221 4.5e-05 0.5] ApplyBuff: Provide the giant a buff to its power level.
  (Ability|HasAuthority
    (:False
      (Utilities|FlowControl|DeltaGateWildcard (Variables|Transient|GetbBuffApplied) true
        (:Success
          (Fortnite|FortShippingLog self "Applying Buff")
          ;;[0.03102 0.08319 0.15 0.5] Handle behavior during the buff moment.
          (if (ScalableFloat|GetValueasBool (Variables|Hotfix|GetHotfix_BuffGameplayEnabled))
            (bind _returnvalue (Ability|ApplyGameplayEffectToOwner (Variables|Settings|GetFinisherBuffGameplayEffectClass)))
            (Variables|Transient|SetFinisherBuffEffectHandle _returnvalue)
            (Ability|HasAuthority
              (:True
                (bind _returnvalue_2 (Utilities|Time|SetTimerbyEvent (EventDispatchers|CreateEvent :!event Custom|HandlePowerLevelRecoveryInterval) (ScalableFloat|GetCachedScalableFloatValue (Variables|Hotfix|GetHotfix_PowerLevelRecoveryInterval)) true))
                (Variables|Transient|SetPowerLevelRecoveryTimerHandle _returnvalue_2)
                (CallFunction|HandlePowerLevelRecoveryInterval)))))))))

(event Custom|ApplyPowerUp
  ; Apply PowerUp
  (Ability|HasAuthority
    (:False
      (Utilities|FlowControl|DeltaGateWildcard (Variables|Transient|GetPowerUpApplied) true
        (:Success
          (Fortnite|FortShippingLog self "Apply Powerup")
          ;;[0.003337 0.06572 0.15 0.5] Start powering up the sword here.
          (Ability|ExecuteGameplayCueOnOwner (Utilities|Struct|BreakGameplayCueTag (Variables|GameplayCueTag|GetGameplayCueTag_SwordPowerup))))))))

(event Custom|ServerApplyFinisher
  (bind _returnvalue (Fortnite|Athena|GameState|GetGameStateAthena))
  (bind _returnvalue_2 (Actor|GetComponentbyClass _returnvalue "/ToyFoxGameplay/Gameplay/B_ToyFoxGameStateComponent.B_ToyFoxGameStateComponent_C"))
  ;[0 0.09131 0.15 0.5] ApplyFinisher: If finisher hasn't been applied yet, apply it to the target actor and proceed with desired behavior.
  (Ability|HasAuthority
    (:True
      (if (Ability|GameplayAbility|IsActive 0)
        (Utilities|FlowControl|DeltaGateWildcard (Variables|Transient|GetbFinisherApplied) true
          (:Success
            (Fortnite|FortShippingLog self "Applying Finisher")
            (Ability|ApplyGameplayEffectToTarget (Ability|TargetData|AbilityTargetDatafromActor (Variables|Transient|GetFinisherTargetActor)) "/TidalCrane_ToyFox_Boss/Abilities/Death/GE_ToyFox_Boss_Death.GE_ToyFox_Boss_Death_C")
            (Ability|ExecuteGameplayCueWithParamsOnOwner (Utilities|Struct|BreakGameplayCueTag (Variables|GameplayCueTag|GetGameplayCueTag_SwordPowerup_Impact)) (Ability|GameplayCue|MakeGameplayCueParameters :NormalizedMagnitude 0.0 :RawMagnitude 0.0 :Location "0,0,0" :Normal "0,0,0" :EffectCauser (Variables|Transient|GetFinisherTargetActor) :GameplayEffectLevel 1 :AbilityLevel 1 :bReplicateLocationWhenUsingMinimalRepProxy false))
            (bind _asfort_player_pawn_athena (Utilities|Casting|CastToFortPlayerPawnAthena (CallFunction|GetAvatarasFortPlayerPawn)))
            (EventDispatchers|CallDisplayHitNotify :self _asfort_player_pawn_athena :Damage 99999.0 :Critical true :Fatal true :HitPawn (Variables|Transient|GetFinisherTargetActor) :HitLocation (+ (Transformation|GetActorLocation (Variables|Transient|GetFinisherTargetActor)) "0.000000,0.000000,2500.000000"))
            ;;[0.15 0.05969 0.1468 0.5] Perform damage to any actors in the way.
            (bind (_asynctaskproxy _targetdata) (Ability|Tasks|PerformTargetingAbilityTask "DamageTargeting" (Variables|Settings|GetDamageTargetingPreset) true false true)
              (:then
                ;;[0.15 0.08762 0.123 0.5] Analytics notify.
                (Ability|HasAuthority
                  (:True
                    (Utilities|IsValid _returnvalue
                      (:Is%20Valid
                        (Utilities|IsValid _returnvalue_2
                          (:Is%20Valid
                            (Analytics|Notify|NotifyGiantFinisherExecuted _returnvalue_2))))))))
              (:Targeted
                _)
              (:Cancelled
                _))
            (Ability|Tasks|ApplyGameplayEffectToTargetDataAsync _targetdata (Variables|Settings|GetDamageGameplayEffectClass) 1 1 (Variables|Hotfix|GetHotfix_MaxTargetsToProcessPerTick))))))))

(event Custom|ApplyFinisher
  (bind _returnvalue (Fortnite|Athena|GameState|GetGameStateAthena))
  (bind _returnvalue_2 (Actor|GetComponentbyClass _returnvalue "/ToyFoxGameplay/Gameplay/B_ToyFoxGameStateComponent.B_ToyFoxGameStateComponent_C"))
  ;[0 0.09131 0.15 0.5] ApplyFinisher: If finisher hasn't been applied yet, apply it to the target actor and proceed with desired behavior.
  (Ability|HasAuthority
    (:False
      (Utilities|FlowControl|DeltaGateWildcard (Variables|Transient|GetbFinisherApplied) true
        (:Success
          (Fortnite|FortShippingLog self "Applying Finisher")
          (Ability|ApplyGameplayEffectToTarget (Ability|TargetData|AbilityTargetDatafromActor (Variables|Transient|GetFinisherTargetActor)) "/TidalCrane_ToyFox_Boss/Abilities/Death/GE_ToyFox_Boss_Death.GE_ToyFox_Boss_Death_C")
          (Ability|ExecuteGameplayCueWithParamsOnOwner (Utilities|Struct|BreakGameplayCueTag (Variables|GameplayCueTag|GetGameplayCueTag_SwordPowerup_Impact)) (Ability|GameplayCue|MakeGameplayCueParameters :NormalizedMagnitude 0.0 :RawMagnitude 0.0 :Location "0,0,0" :Normal "0,0,0" :EffectCauser (Variables|Transient|GetFinisherTargetActor) :GameplayEffectLevel 1 :AbilityLevel 1 :bReplicateLocationWhenUsingMinimalRepProxy false))
          (bind _asfort_player_pawn_athena (Utilities|Casting|CastToFortPlayerPawnAthena (CallFunction|GetAvatarasFortPlayerPawn)))
          (EventDispatchers|CallDisplayHitNotify :self _asfort_player_pawn_athena :Damage 99999.0 :Critical true :Fatal true :HitPawn (Variables|Transient|GetFinisherTargetActor) :HitLocation (+ (Transformation|GetActorLocation (Variables|Transient|GetFinisherTargetActor)) "0.000000,0.000000,2500.000000"))
          ;;[0.15 0.05969 0.1468 0.5] Perform damage to any actors in the way.
          (bind (_asynctaskproxy _targetdata) (Ability|Tasks|PerformTargetingAbilityTask "DamageTargeting" (Variables|Settings|GetDamageTargetingPreset) true false true)
            (:then
              ;;[0.15 0.08762 0.123 0.5] Analytics notify.
              (Ability|HasAuthority
                (:True
                  (Utilities|IsValid _returnvalue
                    (:Is%20Valid
                      (Utilities|IsValid _returnvalue_2
                        (:Is%20Valid
                          (Analytics|Notify|NotifyGiantFinisherExecuted _returnvalue_2))))))))
            (:Targeted
              _)
            (:Cancelled
              _))
          (Ability|Tasks|ApplyGameplayEffectToTargetDataAsync _targetdata (Variables|Settings|GetDamageGameplayEffectClass) 1 1 (Variables|Hotfix|GetHotfix_MaxTargetsToProcessPerTick)))))))

(event Custom|HandlePowerLevelRecoveryInterval
  ;[0.15 0.01408 0 0.5] HandlePowerLevelRecoveryInterval: Increase the power level by the desired amount.
  (Ability|HasAuthority
    (:True
      (bind _output_get (Variables|Transient|SetPowerLevelRecoveryTotalTime (+ (Variables|Transient|GetPowerLevelRecoveryTotalTime) (ScalableFloat|GetCachedScalableFloatValue (Variables|Hotfix|GetHotfix_PowerLevelRecoveryInterval)))))
      (if (> _output_get (ScalableFloat|GetCachedScalableFloatValue (Variables|Hotfix|GetHotfix_PowerLevelRecoveryDuration)))
        (Utilities|Time|ClearandInvalidateTimerbyHandle (Variables|Transient|GetPowerLevelRecoveryTimerHandle))
        (else
          (bind _gamestatecomponent (Variables|Transient|GetGameStateComponent)
            (:then
              (PowerLevel|IncreasePowerLevel _gamestatecomponent (ScalableFloat|GetCachedScalableFloatValue (Variables|Hotfix|GetHotfix_PowerLevelRecoveryAmountPerInterval))))
            (:else)))))))

(event Ability|EventOnEndAbility
  (params (bWasCancelled :type Boolean))
  ;[0.15 0.09929 0 0.5] EndAbility: Clean up.
  (Utilities|FlowControl|Sequence
    (:then_0
      (Ability|RemoveGameplayEffectFromOwnerWithHandle (Variables|Transient|GetFinisherBuffEffectHandle))
      (Ability|RemoveGameplayEffectFromOwnerWithHandle (Variables|Transient|GetFinisherActiveGameplayEffectHandle))
      (Fortnite|FortShippingLog self "EndAbility: Invoked.")
      (Utilities|Time|ClearandInvalidateTimerbyHandle (Variables|Transient|GetPowerLevelRecoveryTimerHandle))
      (bind _asynctask_finishertargeting (Variables|Transient|GetAsyncTask_FinisherTargeting)
        (:then
          (GameplayTasks|EndTask _asynctask_finishertargeting))
        (:else
          _))
      (Ability|HasAuthority
        (:True
          (bind _gamestatecomponent (Variables|Transient|GetGameStateComponent)
            (:then
              (Utilities|FlowControl|Sequence
                (:then_0
                  (Class|BToyFoxGameStateComponent|BroadcastAlertGiantFinisherComplete _gamestatecomponent)
                  (if (not (Utilities|Enum|Equal(Enum) (Event|GetEventState (Variables|Transient|GetGameStateComponent)) "NewEnumerator4"))
                    (Class|BToyFoxGameStateComponent|SetSecondaryPilot3PView (Variables|Transient|GetGameStateComponent) false)))
                (:then_1
                  ;;[0.02567 0 0.15 0.5] Queen death quest for pilot players.
                  (for _array_element (Pilots|GetSecondaryPilotPlayerPawns _gamestatecomponent)
                    (Utilities|IsValid _array_element
                      (:Is%20Valid
                        (CustomVerb|BroadcastCustomVerbMessage (CustomVerb|MakeCustomVerbMessage _array_element "(TagName=\"CustomVerb.DamagedQueen\")")))))
                  (Utilities|IsValid (Giant|GetGiantPawn _gamestatecomponent)
                    (:Is%20Valid
                      (CustomVerb|BroadcastCustomVerbMessage (CustomVerb|MakeCustomVerbMessage (Giant|GetGiantPawn _gamestatecomponent) "(TagName=\"CustomVerb.DamagedQueen\")"))
                      ;;[0.02567 0 0.15 0.5] Queen death event for Military Rank point grant
                      (CustomVerb|BroadcastCustomVerbMessage (CustomVerb|MakeCustomVerbMessage (Giant|GetGiantPawn _gamestatecomponent) "(TagName=\"CustomVerb.ToyFox.DefeatedQueen\")")))))))
            (:else)))))
    (:then_1
      (|CallParentFunction bWasCancelled))))

(fn FallbackTargetDetection
  (bind _returnvalue (Fortnite|Athena|GameState|GetGameStateAthena))
  (bind _returnvalue_2 (Actor|GetComponentbyClass _returnvalue "/ToyFoxGameplay/Gameplay/B_ToyFoxGameStateComponent.B_ToyFoxGameStateComponent_C"))
  (bind _enemy (Enemy|GetEnemy _returnvalue_2))
  ;[0.1047 0 0.15 0.5] FallbackTargetDetection: Simple distance and tag check to verify if the expected target is valid for finisher ability.
  (Utilities|FlowControl|Sequence
    (:then_0
      (Ability|HasAuthority
        (:True
          (if (> (ScalableFloat|GetValueatLevel (Variables|Hotfix|GetHotfix_FallbackTargetCheckDistance)) 0)
            (Utilities|IsValid _returnvalue
              (:Is%20Valid
                (Utilities|IsValid _returnvalue_2
                  (:Is%20Valid
                    (Utilities|IsValid _enemy
                      (:Is%20Valid
                        (if (and (< (Transformation|GetHorizontalDistanceTo _enemy (CallFunction|GetAvatarasFortPlayerPawn)) (ScalableFloat|GetValueatLevel (Variables|Hotfix|GetHotfix_FallbackTargetCheckDistance))) (GameplayTags|HasAllMatchingGameplayTags _enemy (Variables|Settings|GetFinisherValidGameplayTagContainer)))
                          (return true _enemy))))))))))))
    (:then_1
      (return false 0))))
